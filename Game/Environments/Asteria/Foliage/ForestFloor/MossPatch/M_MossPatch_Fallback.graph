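# Fallback masked, default-lit material for forest-floor moss patches in Asteria, enabled for Nanite and instanced static meshes.
# Vertex color red blends the diffuse toward a BottomBrightness-scaled copy, darkened by default, and blends between two opacity multipliers.
# The SMR texture supplies the opacity mask in green and parameter-scaled roughness in blue.
# A normal map and scalar specular complete the Material Attributes output.
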Material /Game/Environments/Asteria/Foliage/ForestFloor/MossPatch/M_MossPatch_Fallback
graph {
 properties {
  bUseMaterialAttributes=true
  bUsedWithInstancedStaticMeshes=true
  bUsedWithNanite=true
  blendMode=BLEND_Masked
  pixelDepthOffsetMode=PDOM_Legacy
  usedShadingModels=Default Lit
 }
 nodes [
  n1 = TextureSample() {texture=&/MossPatch_D}
  n2 = ScalarParameter() {defaultValue=0.512, parameterName=BottomBrightness}
  n3 = VertexColor()
  n4 = TextureSample() {texture=&/MossPatch_SMR}
  n5 = ScalarParameter() {defaultValue=1.0569, parameterName=Opacity1}
  n6 = ScalarParameter() {defaultValue=2.6252, parameterName=Opacity2}
  SetMaterialAttributes!(
   MaterialAttributes
   Base Color=LinearInterpolate(A=@n1.RGB, B=Multiply(A=@n2, B=@n1.RGB), Alpha=@n3.R)
   Specular=ScalarParameter() {defaultValue=0.5, parameterName=Specular}
   Roughness=Multiply(A=@n4.B, B=ScalarParameter() {defaultValue=2.4581, parameterName=Roughness})
   Normal=TextureSample() {samplerType=SAMPLERTYPE_Normal, texture=&/MossPatch_N}.RGB
   Opacity Mask=Multiply(A=@n4.G, B=LinearInterpolate(A=@n5, B=@n6, Alpha=@n3.R))
  ) {attributeSetTypes [
    #1
    #2
    #3
    #4
    #5
   ]}
 ]
}
